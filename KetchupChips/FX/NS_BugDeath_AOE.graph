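# Niagara one-shot area-of-effect burst for a bug death, built from a purple glow, an expanding mesh sphere and a fading ground decal.
# NE_StarGlow derives separate glow, dark glow, corona and shockwave colors and sizes to feed several sprite renderers; its light module is disabled.
# User.FXScale drives most sizes, FXOpacity scales most fades, and FXDecal gates decal spawning through spawn probability.
# Flares and the debug radius ring emitters are disabled.
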NiagaraSystem /KetchupChips/FX/NS_BugDeath_AOE
dependencies {
 usedDataInterfaces [
  &NiagaraDataInterfaceCurve
  &NiagaraDataInterfaceSpriteRendererInfo
  &NiagaraDataInterfaceColorCurve
  &NiagaraDataInterfaceVector2DCurve
 ]
 usedDynamicInputs []
 usedModules [
  &SystemState
  &self:SystemSpawnScript.SystemScriptSource.SystemScriptGraph.NiagaraNodeAssignment_0.SetVariables_#1
 ]
}
emitters {
 DEBUG {
  bEnabled=false
  emitterUpdateScript [
   &EmitterState()
   &SpawnBurst_Instantaneous(
    Spawn Count=1000
    Spawn Probability=&MakeCustomFloatFromBool(Bool=$User.ShowDebugRing, True Float=1, False Float=0)
    Use Spawn Probability=true
    UseLoopCountLimit=false
   )
  ]
  particleSpawnScript [
   &InitializeParticle(Color=<rgba 0.4992 0 1 1>, Lifetime=0.75, Sprite Size Mode=::NewEnumerator3, UsePositionOffset=false, Write Material Random=true)
   &ShapeLocation(MaskSpawnGroup=false, Override Seed=false, Ring Radius=$User.FXScale, Shape Primitive=::NewEnumerator4, UseOwnerScale=true)
  ]
  particleUpdateScript [&ParticleState()]
  rendererClasses [&NiagaraSpriteRendererProperties]
  renderers [{rendererClass=&NiagaraSpriteRendererProperties}]
 }
 DecalRenderer {
  emitterUpdateScript [
   &EmitterState()
   &SpawnBurst_Instantaneous(
    Spawn Probability=&MakeCustomFloatFromBool(Bool=$User.FXDecal, True Float=1, False Float=0)
    Use Spawn Probability=true
    UseLoopCountLimit=false
   )
   /SetVariables(
    Emitter.OpacityMultCrack:NiagaraFloat=&RandomRangeFloat(Minimum=0.2, Maximum=0.5)
    Emitter.RandRotTex:NiagaraFloat=&RandomRangeFloat(Minimum=0, Maximum=1)
   ) {isSetParameters=true}
  ]
  particleSpawnScript [
   &InitializeParticle(Lifetime=0.55, UsePositionOffset=true, Write Material Random=true)
   /SetVariables(Particles.DecalOrientationRing:Quat4f=/Game/Effects/Niagara/Modules/OrientMestToVectorFromCustomRot(
     InitialRotation:Quat4f=<0 0 0 1>
     Look At Direction:Vector3f=<0 0 0>
     Reference Axis Vector:Vector3f=<0 0 1>
     RotationVector:Vector3f=&RandomRangeVector(Minimum=<0 0 0>, Maximum=<1 0 0>)
    )) {isSetParameters=true}
   /SetVariables(Particles.DecalPosSpray:NiagaraPosition=&AddVectorToPosition(Position=$Particles.Position, Vector=<0 0 0>)) {enabled=false, isSetParameters=true}
  ]
  particleUpdateScript [
   &ParticleState()
   /SetVariables(Particles.DecalFadeSpray:NiagaraFloat=&Lerp_Float(
     A=0
     B=&Multiply_Float(A=1, B=$User.FXOpacity)
     Alpha=&FloatFromCurve(FloatCurve=DataInterface() {propertyValues {Curve {keys [
          {arriveTangent=41.722, time=0, value=1}
          {interpMode=RCIM_Cubic, tangentMode=RCTM_User, time=0.2, value=1}
          {arriveTangent=-0.5041, interpMode=RCIM_Cubic, leaveTangent=-0.5041, tangentMode=RCTM_User, time=1, value=0}
         ]}}})
    )) {isSetParameters=true}
   /SetVariables(Particles.DecalSizeSpray:Vector3f=&Multiply_VectorByFloat(
     Vector=&Lerp_Vector(
      A=<100 100 100>
      B=<100 575 575>
      Alpha=&FloatFromCurve(FloatCurve=DataInterface() {propertyValues {Curve {keys [
           {interpMode=RCIM_Cubic, leaveTangentWeight=0.2519, tangentMode=RCTM_User, time=0, value=0.15}
           {arriveTangent=2.3559, interpMode=RCIM_Cubic, leaveTangent=2.3559, tangentMode=RCTM_User, time=0.2, value=0.65}
           {interpMode=RCIM_Cubic, tangentMode=RCTM_User, time=0.5, value=1}
          ]}}})
     )
     Float=&Multiply_Float(A=0.004, B=$User.FXScale)
    )) {isSetParameters=true}
  ]
  rendererClasses [&NiagaraDecalRendererProperties]
  renderers [{rendererClass=&NiagaraDecalRendererProperties, rendererIndex=0}]
 }
 Flares {
  bEnabled=false
  emitterUpdateScript [
   &EmitterState(Life Cycle Mode=::NewEnumerator1, Loop Behavior=::NewEnumerator1, Loop Delay=0.011, Loop Duration=1.65, UseLoopDelay=true)
   &SpawnBurst_Instantaneous(Spawn Count=8, Use Spawn Probability=false, UseLoopCountLimit=false)
  ]
  particleSpawnScript [
   &InitializeParticle(
    Color=&MakeLinearColorFromVectorAndFloat("Vector (RGB)"=<1 1 1>, "Float (Alpha)"=&RandomRangeFloat(Minimum=0.9, Maximum=1))
    Lifetime=&RandomRangeFloat(Minimum=0.35, Maximum=0.43)
    Sprite Size Max=<250 500>
    Sprite Size Min=<200 400>
    Sprite Size Mode=::NewEnumerator2
    UsePositionOffset=false
    Write Material Random=true
   )
   &ShapeLocation(
    MaskSpawnGroup=false
    Sphere Distribution=::NewEnumerator1
    Sphere Radius=20
    U Position=&Lerp_Float(A=0, B=1, Alpha=&ReturnNormalizedExecIndex(Normalized Index Scale=2))
    UseOwnerScale=true
    V Position=&RandomRangeFloat(Minimum=0.5, Maximum=1)
   )
   &CameraOffset(Camera Offset Amount=4)
   &SubUVAnimation(
    Evaluation Type=::NewEnumerator0
    "Sprite Renderer "=DataInterface() {propertyValues {SpriteRenderer=&self:Flares_0.NiagaraSpriteRendererProperties_2}}
    SubUV Animation Mode:ENiagara_SubUVLookupModeV2=::NewEnumerator2
    UseEndFrame=false
    UseStartFrame=false
   )
   &AddVelocity(Velocity=&NormalizeVector(Vector To Normalize=$Particles.ShapeLocation.ShapeNormal, Safe Normalize=true, Safe Fallback Threshold=0, Safe Fallback Vector=<0 0 0.001>))
  ]
  particleUpdateScript [
   &ParticleState()
   &ScaleColor(
    Scale Alpha=&FloatFromCurve(
     FloatCurve=DataInterface() {propertyValues {Curve {keys [
         {arriveTangent=12.7361, interpMode=RCIM_Cubic, leaveTangent=12.7361, tangentMode=RCTM_User, time=0, value=0}
         {arriveTangent=-2.2233, interpMode=RCIM_Cubic, leaveTangent=-2.2233, tangentMode=RCTM_User, time=0.1, value=1}
         {arriveTangent=-0.0075, interpMode=RCIM_Cubic, leaveTangent=-0.0075, tangentMode=RCTM_User, time=1, value=0}
        ]}}}
     Scale Curve=1.1
    )
    Scale RGB=&MakeVectorFromLinearColorRGB(Color=&ColorFromCurve(ColorCurve=DataInterface() {propertyValues {
        AlphaCurve {keys [<at 0 1 RCIM_Cubic>]}
        BlueCurve {keys [<at 0.0426 0.6741 RCIM_Cubic>, {interpMode=RCIM_Cubic, leaveTangent=0.0077, time=0.2988, value=0.4441}]}
        GreenCurve {keys [<at 0.0426 3.8339 RCIM_Cubic>, {interpMode=RCIM_Cubic, leaveTangent=-5.955, time=0.2988, value=2.2476}]}
        RedCurve {keys [<at 0.0426 10 RCIM_Cubic>, {interpMode=RCIM_Cubic, leaveTangent=-12.0954, time=0.2988, value=5}]}
       }}))
    ScaleA=true
    ScaleRGB=true
   )
   &ScaleSpriteSize(
    Scale Sprite Size Mode=::NewEnumerator0
    Uniform Scale Factor=&FloatFromCurve(FloatCurve=DataInterface() {propertyValues {Curve {keys [
         {arriveTangent=3.0498, interpMode=RCIM_Cubic, leaveTangent=3.0498, tangentMode=RCTM_User, time=0, value=0.3}
         {arriveTangent=-0.026, interpMode=RCIM_Cubic, leaveTangent=-0.026, tangentMode=RCTM_User, time=0.5, value=1}
        ]}}})
   )
   &ScaleSpriteSize(
    Non-Uniform Curve Sprite Scale=DataInterface() {propertyValues {
      XCurve {keys [
        {arriveTangent=-1.6611, interpMode=RCIM_Cubic, leaveTangent=-1.6611, tangentMode=RCTM_User, time=0.2, value=1.2}
        {arriveTangent=-0.1324, interpMode=RCIM_Cubic, leaveTangent=-0.1324, tangentMode=RCTM_User, time=1, value=0.7}
       ]}
      YCurve {keys [{interpMode=RCIM_Cubic, tangentMode=RCTM_User, time=0.4, value=1}]}
     }}
    Scale Sprite Size Mode=::NewEnumerator3
   ) {module=ScaleSpriteSize001}
   &SolveForcesAndVelocity(Clamp Velocity=false, Limit Acceleration=false)
   &DynamicMaterialParameters(
    Index 0 Param 2=1
    Index 0 Param 3=1
    Index 0 Param 4=1
    Module.Index0Param0Write=true
    Module.Index0Param1Write=true
    Module.Index0Param2Write=true
    Module.Index0Param3Write=true
   )
  ]
  rendererClasses [&NiagaraSpriteRendererProperties]
  renderers [{rendererClass=&NiagaraSpriteRendererProperties}]
 }
 NE_StarGlow {
  emitterUpdateScript [
   &EmitterState()
   &SpawnBurst_Instantaneous(Use Spawn Probability=false, UseLoopCountLimit=false)
  ]
  particleSpawnScript [
   &InitializeParticle(
    Color=<rgba 0.5401 0.0002 1 1>
    Lifetime=0.5
    Mass Mode=::NewEnumerator3
    Position Offset=<0 0 10>
    Sprite Rotation Mode=::NewEnumerator2
    Sprite Size Mode=::NewEnumerator3
    Uniform Sprite Size=900
    UsePositionOffset=false
    Write Material Random=true
   )
   &CameraOffset(Camera Offset Amount=4) {enabled=false}
   &SpriteFacingAndAlignment(Sprite Alignment=<1 0 0>, Sprite Facing=<0 0 1>, WriteAlignment=true, WriteFacing=true)
  ]
  particleUpdateScript [
   &ParticleState()
   &SpriteRotationRate(Rotation Rate=&FloatFromCurve(
     FloatCurve=DataInterface() {propertyValues {Curve {keys [
         {arriveTangent=1.4306, interpMode=RCIM_Cubic, leaveTangent=1.4306, tangentMode=RCTM_User, time=0.2, value=0}
         {arriveTangent=1.1179, interpMode=RCIM_Cubic, leaveTangent=1.1179, tangentMode=RCTM_User, time=0.4, value=0.6228}
         {arriveTangent=0.0224, interpMode=RCIM_Cubic, leaveTangent=0.0224, tangentMode=RCTM_User, time=1, value=1}
        ]}}}
     Scale Curve=10
    ))
   &ScaleColor(
    Scale Alpha=&FloatFromCurve(FloatCurve=DataInterface() {propertyValues {Curve {keys [<at 0 1>,<at 1 0>]}}})
    ScaleA=true
    ScaleRGB=false
   )
   &ScaleSpriteSize(
    Scale Sprite Size Mode=::NewEnumerator0
    Uniform Scale Factor=&FloatFromCurve(FloatCurve=DataInterface() {propertyValues {Curve {keys [
         {arriveTangent=2.3574, interpMode=RCIM_Cubic, leaveTangent=2.3574, tangentMode=RCTM_User, time=0, value=0}
         {arriveTangent=0.6516, interpMode=RCIM_Cubic, leaveTangent=0.6516, tangentMode=RCTM_User, time=1, value=1}
        ]}}})
   )
   /SetVariables(
    Particles.GlowColor=&MultiplyLinearColor(
     Linear Color=$Particles.Color
     Scale Factor=&MakeFloat4FromVectorAndFloat("Vector (RGB)"=&VectorFromFloat(Value=1), "Float (Alpha)"=0.5)
    )
    Particles.GlowSize=&Multiply_Vector2D_ByFloat(Float=&Multiply_Float(A=0.007, B=$User.FXScale), Vector2D=$Particles.SpriteSize)
   ) {isSetParameters=true}
   /SetVariables(
    Particles.DarkGlowColor:LinearColor=&MultiplyLinearColor(
     Linear Color=$Particles.Color
     Scale Factor=&MakeVector4FromVectorAndFloat("Vector (RGB)"=<0.15 0.15 0.15>, "Float (Alpha)"=&Multiply_Float(A=0.5, B=$User.FXOpacity))
    )
    Particles.DarkGlowSize:Vector2f=&Multiply_Vector2D_ByFloat(Float=&Multiply_Float(A=0.01, B=$User.FXScale), Vector2D=$Particles.SpriteSize)
   ) {isSetParameters=true}
   /SetVariables(
    Particles.CoronaColor:LinearColor=&MultiplyLinearColor(
     Linear Color=$Particles.Color
     Scale Factor=&MakeFloat4FromVectorAndFloat(
      "Vector (RGB)"=&VectorFromFloat(Value=3)
      "Float (Alpha)"=&FloatFromCurve(
       FloatCurve=DataInterface() {propertyValues {Curve {keys [
           <at 0.1 0 RCIM_Cubic>
           {interpMode=RCIM_Cubic, tangentMode=RCTM_User, time=0.17, value=1}
           {interpMode=RCIM_Cubic, tangentMode=RCTM_User, time=0.6, value=0}
          ]}}}
       Scale Curve=$User.FXOpacity
      )
     )
    )
    Particles.PositionCorona:NiagaraPosition=&AddVectorToPosition(Position=$Engine.Emitter.SimulationPosition, Vector=<0 0 0>)
    Particles.SpriteSizeCorona:Vector2f=&Multiply_Vector2D_ByFloat(
     Float=&FloatFromCurve(FloatCurve=DataInterface() {propertyValues {Curve {keys [
          {arriveTangent=1.7172, interpMode=RCIM_Cubic, leaveTangent=1.7172, tangentMode=RCTM_User, time=0.12, value=0.42}
          {interpMode=RCIM_Cubic, tangentMode=RCTM_User, time=0.9, value=1}
         ]}}})
     Vector2D=&Vector2DFromFloat(Value=&Multiply_Float(A=3.7, B=$User.FXScale))
    )
   ) {isSetParameters=true}
   /SetVariables(
    Particles.ColorShockwave:LinearColor=&MultiplyLinearColor(
     Linear Color=$Particles.Color
     Scale Factor=&MakeVector4(
      X=1
      Y=1
      Z=1
      W=&FloatFromCurve(FloatCurve=DataInterface() {propertyValues {Curve {keys [{interpMode=RCIM_Cubic, tangentMode=RCTM_User, time=0.3, value=1}, {interpMode=RCIM_Cubic, tangentMode=RCTM_User, time=0.7, value=0}]}}})
     )
    )
    Particles.SpriteSizeShockWave:Vector2f=&Multiply_Vector2D_ByFloat(
     Float=&FloatFromCurve(FloatCurve=DataInterface() {propertyValues {Curve {keys [
          {arriveTangent=1.4072, interpMode=RCIM_Cubic, leaveTangent=1.4072, tangentMode=RCTM_User, time=0, value=0.6}
          {arriveTangent=0.2948, interpMode=RCIM_Cubic, leaveTangent=0.2948, tangentMode=RCTM_User, time=0.4, value=0.95}
          {interpMode=RCIM_Cubic, tangentMode=RCTM_User, time=0.9, value=1}
         ]}}})
     Vector2D=&Vector2DFromFloat(Value=300)
    )
   ) {isSetParameters=true}
   &FN_Light_Attributes(
    Color=$Particles.Color
    Create Color Attribute=true
    Create Exponent Attribute=true
    Create Radius Attribute=true
    Radius=&Multiply_Float(A=1.4, B=$User.FXScale)
    Scale Color=&FloatFromCurve(
     FloatCurve=DataInterface() {propertyValues {Curve {keys [
         {interpMode=RCIM_Cubic, tangentMode=RCTM_User, time=0, value=1}
         {arriveTangent=-1.742, interpMode=RCIM_Cubic, leaveTangent=-1.742, tangentMode=RCTM_User, time=1, value=0}
        ]}}}
     Scale Curve=10
    )
   ) {enabled=false}
  ]
  rendererClasses [&NiagaraSpriteRendererProperties, &NiagaraLightRendererProperties]
  renderers [
   {rendererClass=&NiagaraSpriteRendererProperties}
   {rendererClass=&NiagaraSpriteRendererProperties, rendererIndex=1}
   {rendererClass=&NiagaraLightRendererProperties, rendererIndex=2}
   {rendererClass=&NiagaraSpriteRendererProperties, rendererIndex=3}
   {rendererClass=&NiagaraSpriteRendererProperties, rendererIndex=4}
   {rendererClass=&NiagaraSpriteRendererProperties, rendererIndex=5}
  ]
 }
 SphereMain {
  emitterUpdateScript [
   &EmitterState()
   &SpawnBurst_Instantaneous(Use Spawn Probability=false, UseLoopCountLimit=false)
   /SetVariables(Emitter.PlasmaAmount=1, Emitter.UVScale:Vector4f=<0.3 0.3 0.3 0.3>) {isSetParameters=true}
  ]
  particleSpawnScript [
   &InitializeParticle(
    Color=<rgba 0.4414 0 1 1>
    Lifetime=0.55
    Mesh Scale Mode=::NewEnumerator3
    Mesh Uniform Scale=&Multiply_Float(A=0.045, B=$User.FXScale)
    UsePositionOffset=false
    Write Material Random=true
   )
   &InitialMeshOrientation(Mesh Orientation Mode=::NewEnumerator2, Use Rotation Vector=false)
  ]
  particleUpdateScript [
   &ParticleState()
   &ScaleMeshSize(Scale Factor=&VectorFromFloat(Value=&Lerp_Float(
      A=0.5
      B=1
      Alpha=&FloatFromCurve(FloatCurve=DataInterface() {propertyValues {Curve {keys [
           {interpMode=RCIM_Cubic, leaveTangentWeight=0.9987, tangentMode=RCTM_User, time=0, value=0}
           {interpMode=RCIM_Cubic, tangentMode=RCTM_User, time=1, value=1}
          ]}}})
     )))
   &DynamicMaterialParameters(
    Index 0 Param 1=&FloatFromCurve(FloatCurve=DataInterface() {propertyValues {Curve {keys [<at 0 0 RCIM_Cubic>, {arriveTangentWeight=1.205, interpMode=RCIM_Cubic, tangentMode=RCTM_User, time=1, value=1}]}}})
    Module.Index0Param0Write=true
    Module.Index0Param1Write=false
    Module.Index0Param2Write=true
    Module.Index0Param3Write=true
   )
   &ScaleColor(
    Scale Alpha=&FloatFromCurve(
     FloatCurve=DataInterface() {propertyValues {Curve {keys [
         {interpMode=RCIM_Cubic, tangentMode=RCTM_User, time=0, value=0}
         <at 0.2 1 RCIM_Cubic>
         {arriveTangentWeight=0.6641, interpMode=RCIM_Cubic, tangentMode=RCTM_User, time=1, value=0}
        ]}}}
     Scale Curve=&Multiply_Float(A=0.75, B=$User.FXOpacity)
    )
    Scale RGB=&VectorFromFloat(Value=&Multiply_Float(
      A=&Lerp_Float(
       A=1
       B=3
       Alpha=&FloatFromCurve(FloatCurve=DataInterface() {propertyValues {Curve {keys [
            {interpMode=RCIM_Cubic, leaveTangentWeight=0.1521, tangentMode=RCTM_User, time=0, value=1}
            {arriveTangent=-0.0405, interpMode=RCIM_Cubic, leaveTangent=-0.0405, tangentMode=RCTM_User, time=1, value=0}
           ]}}})
      )
      B=1
     ))
    ScaleA=true
    ScaleRGB=true
   )
   &GravityForce()
   &SolveForcesAndVelocity(Clamp Velocity=false, Limit Acceleration=false)
  ]
  rendererClasses [&NiagaraMeshRendererProperties]
  renderers [{rendererClass=&NiagaraMeshRendererProperties, rendererIndex=0}]
 }
}
summary {
 systemName=NS_BugDeath_AOE
 emitters [
  {emitterName=NE_StarGlow, rendererClasses [&NiagaraSpriteRendererProperties, &NiagaraLightRendererProperties]}
  {bEnabled=false, emitterName=Flares, rendererClasses [&NiagaraSpriteRendererProperties]}
  {bEnabled=false, emitterName=DEBUG, rendererClasses [&NiagaraSpriteRendererProperties]}
  {emitterName=SphereMain, rendererClasses [&NiagaraMeshRendererProperties]}
  {emitterName=DecalRenderer, rendererClasses [&NiagaraDecalRendererProperties]}
 ]
 userVariables [
  User.FXDecal {type=NiagaraBool, defaultValue=true}
  User.FXOpacity {type=NiagaraFloat, defaultValue=1, description=""}
  User.FXScale {type=NiagaraFloat, defaultValue=350, description=""}
  User.ShowDebugRing {type=NiagaraBool}
 ]
}
