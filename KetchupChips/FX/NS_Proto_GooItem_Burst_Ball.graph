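# Niagara burst effect for a prototype goo item: a liquid splash of sprite streaks, splashes, drops and mesh juice.
# Projectile particles feed a ribbon emitter through particle attribute reads, forming trails that thin with parent age.
# User FXColor indexes color arrays in most emitters, FXScale sizes many elements, and a time-of-day scalar adjusts some alphas.
# FXDecal gates a decal emitter with three renderers and animated size and fade.
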NiagaraSystem /KetchupChips/FX/NS_Proto_GooItem_Burst_Ball
dependencies {
 usedDataInterfaces [
  &NiagaraDataInterfaceArrayColor
  &NiagaraDataInterfaceCurve
  &NiagaraDataInterfaceParticleRead
  &NiagaraDataInterfaceVector2DCurve
 ]
 usedDynamicInputs [&FNTimeOfDayScalar]
 usedModules [
  &SystemState
  &self:SystemSpawnScript.SystemScriptSource.SystemScriptGraph.NiagaraNodeAssignment_0.SetVariables_#1
 ]
}
emitters {
 BrightStreaks {
  emitterUpdateScript [
   &EmitterState()
   &SpawnBurst_Instantaneous(Spawn Count=7, Use Spawn Probability=false, UseLoopCountLimit=false)
  ]
  particleSpawnScript [
   /SetVariables(Particles.SizeMax=450, Particles.SizeMin=150) {isSetParameters=true}
   /SetVariables(Particles.SizeCurve:NiagaraFloat=&Lerp_Float(
     A=$Particles.SizeMin
     B=$Particles.SizeMax
     Alpha=&FloatFromCurve(
      FloatCurve=DataInterface() {propertyValues {Curve {keys [
          {interpMode=RCIM_Cubic, tangentMode=RCTM_Break, time=0, value=0}
          {arriveTangent=2.305, interpMode=RCIM_Cubic, tangentMode=RCTM_Break, time=1, value=1}
         ]}}}
      CurveIndex=&RandomRangeFloat(Minimum=0, Maximum=1)
     )
    )) {isSetParameters=true}
   &InitializeParticle(
    Color Maximum=<rgba 0.2966 0.055 0.6354 1>
    Color Minimum=<rgba 0.0622 0.0091 0.3646 1>
    Color Mode=::NewEnumerator2
    Lifetime=&Multiply_Float(A=&RandomRangeFloat(Minimum=0.2, Maximum=0.35), B=1)
    Sprite Size=&Multiply_Vector2D_ByFloat(Float=0.35, Vector2D=&RandomRangeVector2D(Minimum=<75 100>, Maximum=<100 210>))
    Sprite Size Mode=::NewEnumerator4
    UsePositionOffset=false
    Write Material Random=true
   )
   &ShapeLocation(MaskSpawnGroup=false, Override Seed=false, Sphere Radius=&Multiply_Float(A=0.5, B=1), UseOwnerScale=true)
   &AddVelocity(
    Module.Use Falloff Distance=false
    Override Seed=false
    Velocity Mode=::NewEnumerator1
    Velocity Speed=&Multiply_Float(
     A=&Lerp_Float(
      A=300
      B=400
      Alpha=&FloatFromCurve(
       FloatCurve=DataInterface() {propertyValues {Curve {keys [{interpMode=RCIM_Cubic, tangentMode=RCTM_Break, time=0, value=0}, {interpMode=RCIM_Cubic, tangentMode=RCTM_Break, time=1, value=1}]}}}
       CurveIndex=&RandomRangeFloat(Minimum=0, Maximum=1)
      )
     )
     B=1
    )
   )
   &SubUVAnimation(End Frame=3, SubUV Animation Mode:ENiagara_SubUVLookupMode=::NewEnumerator2)
   &CameraOffset(Camera Offset Amount=50)
  ]
  particleUpdateScript [
   &ParticleState()
   &Drag(Drag=4, Use Linear Drag=true, Use Rotational Drag=false)
   &SolveForcesAndVelocity(Clamp Velocity=false, Limit Acceleration=false)
   &ScaleColor(
    Scale Alpha=&FloatFromCurve(
     FloatCurve=DataInterface() {propertyValues {Curve {keys [
         {interpMode=RCIM_Cubic, leaveTangentWeight=3.0693, tangentMode=RCTM_Break, time=0, value=1}
         {interpMode=RCIM_Cubic, tangentMode=RCTM_Break, time=1, value=0}
        ]}}}
     Scale Curve=&RandomRangeFloat(Minimum=0.5, Maximum=1.2)
    )
    Scale RGB=&VectorFromFloat(Value=&Lerp_Float(
      A=0.5
      B=50
      Alpha=&FloatFromCurve(FloatCurve=DataInterface() {propertyValues {Curve {keys [
           {interpMode=RCIM_Cubic, tangentMode=RCTM_User, time=0, value=0}
           {arriveTangent=2.2791, interpMode=RCIM_Cubic, leaveTangent=2.2791, tangentMode=RCTM_User, time=1, value=1}
          ]}}})
     ))
    ScaleA=true
    ScaleRGB=true
   )
   &ScaleSpriteSize(Scale Factor=&Vector2DFromCurve(
     Vector2Curve=DataInterface() {propertyValues {XCurve {keys [<at 0.0036 1.6897>,<at 0.9967 0.0582>]}, YCurve {keys [<at 0 1>,<at 1.002 3.7227>]}}}
     ScaleCurve=<1 1>
    ))
   &ScaleSpriteSize(Scale Sprite Size Mode=::NewEnumerator0, Uniform Scale Factor=3) {module=ScaleSpriteSize001}
   &DynamicMaterialParameters(
    Index 0 Param 2=1
    Index 0 Param 3=1
    Index 0 Param 4=&FloatFromCurve(
     FloatCurve=DataInterface() {propertyValues {Curve {keys [{interpMode=RCIM_Cubic, tangentMode=RCTM_Break, time=0.3, value=1}, {interpMode=RCIM_Cubic, tangentMode=RCTM_Break, time=1, value=0.25}]}}}
     Scale Curve=&RandomRangeFloat(Minimum=0.75, Maximum=1)
    )
    Module.Param0WriteEnabled=true
    Module.Param1WriteEnabled=true
    Module.Param2WriteEnabled=true
    Module.Param3WriteEnabled=true
   )
  ]
  rendererClasses [&NiagaraSpriteRendererProperties]
  renderers [{rendererClass=&NiagaraSpriteRendererProperties}]
 }
 CenterSplash {
  emitterUpdateScript [
   &EmitterState()
   &SpawnBurst_Instantaneous(Use Spawn Probability=false, UseLoopCountLimit=false)
   /SetVariables(
    Emitter.FresnelColor:LinearColor=&RandomRangeLinearColor(Minimum=<rgba 0.96 0.022 1 1>, Maximum=<rgba 0.6245 0.0005 1 1>)
    Emitter.TileColor:LinearColor=<rgba 0.2607 0.0126 0.2865 1>
   ) {enabled=false, isSetParameters=true}
   /SetVariables(
    Emitter.FresnelColor:LinearColor=&HueShiftLinearColor(
     Hue=0
     Saturation=1
     Value=&RandomRangeFloat(Minimum=0.5, Maximum=1)
     Linear Color=&SelectLinearColorFromArray(
      Array Sampling Mode=::NewEnumerator1
      Direct Array Index=$User.FXColor
      Color Selection Array=DataInterface() {propertyValues {ColorData [<rgba 1 1 1 1>, <rgba 0 0.6717 1 1>, <rgba 0.96 0.022 1 1>, <rgba 0.1943 1 0 1>]}}
      Random Seed=?
     )
    )
    Emitter.TileColor:LinearColor=&SelectLinearColorFromArray(
     Array Sampling Mode=::NewEnumerator1
     Direct Array Index=$User.FXColor
     Color Selection Array=DataInterface() {propertyValues {ColorData [<rgba 1 1 1 1>, <rgba 0 0.6717 1 1>, <rgba 0.2607 0.0126 0.2865 1>, <rgba 0.022 0.2865 0 1>]}}
     Random Seed=?
    )
   ) {isSetParameters=true}
  ]
  particleSpawnScript [
   &InitializeParticle(
    Lifetime=0.6
    Position Offset=<0 0 10>
    Sprite Rotation Mode=::NewEnumerator2
    Sprite Size Mode=::NewEnumerator1
    Uniform Sprite Size Max=350
    Uniform Sprite Size Min=300
    UsePositionOffset=true
    Write Material Random=true
   )
   &AddVelocity(
    Cone Axis=<0 0 1>
    Module.UseConeFalloff=false
    Module.UseVelDistribution=false
    Override Seed=false
    Velocity Mode=::NewEnumerator2
    Velocity Speed=&RandomRangeFloat(Minimum=100, Maximum=200)
   ) {enabled=false}
   &ShapeLocation(MaskSpawnGroup=false, Override Seed=false, Sphere Radius=10, UseOwnerScale=true)
   &CameraOffset(Camera Offset Amount=35)
  ]
  particleUpdateScript [
   &ParticleState()
   &Drag(Drag=5, Use Linear Drag=true, Use Rotational Drag=false) {enabled=false}
   &DynamicMaterialParameters(
    Index 0 Param 1=&FloatFromCurve(FloatCurve=DataInterface() {propertyValues {Curve {keys [
         <at 0 -1 RCIM_Cubic>
         {arriveTangent=2.3041, interpMode=RCIM_Cubic, leaveTangent=2.3041, tangentMode=RCTM_User, time=0.3, value=-0.675}
         <at 1 0.5>
        ]}}})
    Index 0 Param 2=&FloatFromCurve(FloatCurve=DataInterface() {propertyValues {Curve {keys [<at 0 3>,<at 0.3 1>]}}})
    Module.Param0WriteEnabled=true
    Module.Param1WriteEnabled=true
    Module.Param2WriteEnabled=true
    Module.Param3WriteEnabled=true
   )
   &ScaleSpriteSize(
    Uniform Curve Scale=&Multiply_Float(A=2.2, B=$User.FXScale)
    Uniform Curve Sprite Scale=DataInterface() {propertyValues {Curve {keys [
        {arriveTangent=2.4725, interpMode=RCIM_Cubic, leaveTangent=2.4725, tangentMode=RCTM_User, time=0, value=0.3}
        {arriveTangent=0.6741, interpMode=RCIM_Cubic, leaveTangent=0.6741, tangentMode=RCTM_User, time=0.2, value=0.75}
        {tangentMode=RCTM_User, time=1, value=1}
       ]}}}
   )
   &GravityForce(Gravity=&RandomRangeVector(Minimum=<0 0 -300>, Maximum=<0 0 -400>))
   &SolveForcesAndVelocity(Clamp Velocity=false, Limit Acceleration=false)
   &ScaleColor(Scale Alpha=$System.TOD, Scale RGB=&VectorFromFloat(Value=0.1), ScaleA=false, ScaleRGB=true)
  ]
  rendererClasses [&NiagaraSpriteRendererProperties]
  renderers [{rendererClass=&NiagaraSpriteRendererProperties}]
 }
 DEBUG {
  bEnabled=false
  emitterUpdateScript [
   &EmitterState()
   &SpawnBurst_Instantaneous(Use Spawn Probability=false, UseLoopCountLimit=false)
  ]
  particleSpawnScript [&InitializeParticle(Color=<rgba 0.9908 0 1 1>, Sprite Size Mode=::NewEnumerator3, Uniform Sprite Size=50, UsePositionOffset=false, Write Material Random=true)]
  particleUpdateScript [
   &ParticleState()
   &SolveForcesAndVelocity(Clamp Velocity=false, Limit Acceleration=false)
   &ScaleColor(Scale RGB=&VectorFromFloat(Value=50), ScaleA=false, ScaleRGB=true)
  ]
  rendererClasses [&NiagaraSpriteRendererProperties]
  renderers [{rendererClass=&NiagaraSpriteRendererProperties}]
 }
 DecalRenderer {
  emitterUpdateScript [
   &EmitterState()
   &SpawnBurst_Instantaneous(
    Spawn Probability=&MakeCustomFloatFromBool(Bool=$User.FXDecal, True Float=1, False Float=0)
    Spawn Time=0.05
    Use Spawn Probability=true
    UseLoopCountLimit=false
   )
   /SetVariables(
    Emitter.OpacityMultCrack:NiagaraFloat=&RandomRangeFloat(Minimum=1, Maximum=1.5)
    Emitter.RandRotCracks:NiagaraFloat=&RandomRangeFloat(Minimum=0, Maximum=1)
   ) {isSetParameters=true}
  ]
  particleSpawnScript [
   &InitializeParticle(Lifetime=2, UsePositionOffset=true, Write Material Random=true)
   /SetVariables(Particles.DecalOrientationSpray:Quat4f=/Game/Effects/Niagara/Modules/OrientMestToVectorFromCustomRot(
     InitialRotation:Quat4f=<0 0 0 1>
     Look At Direction:Vector3f=<0 0 0>
     Reference Axis Vector:Vector3f=<0 0 0>
     RotationVector:Vector3f=&RandomRangeVector(Minimum=<0 0 0>, Maximum=<1 0 0>)
    )) {enabled=false, isSetParameters=true}
   /SetVariables(Particles.DecalPosSpray:NiagaraPosition=&AddVectorToPosition(Position=$Particles.Position, Vector=<0 0 0>)) {isSetParameters=true}
  ]
  particleUpdateScript [
   &ParticleState()
   /SetVariables(
    Particles.DecalFadeCrack:NiagaraFloat=&Lerp_Float(
     A=0
     B=1
     Alpha=&FloatFromCurve(FloatCurve=DataInterface() {propertyValues {Curve {keys [
          {arriveTangent=41.722, time=0, value=0}
          {interpMode=RCIM_Cubic, tangentMode=RCTM_User, time=0.02, value=1}
          {arriveTangent=-1.7378, interpMode=RCIM_Cubic, leaveTangent=-1.7378, tangentMode=RCTM_User, time=0.7, value=0.7}
          {arriveTangent=-0.5041, interpMode=RCIM_Cubic, leaveTangent=-0.5041, tangentMode=RCTM_User, time=1, value=0}
         ]}}})
    )
    Particles.DecalFadeScorch:NiagaraFloat=&Lerp_Float(
     A=0
     B=1
     Alpha=&FloatFromCurve(FloatCurve=DataInterface() {propertyValues {Curve {keys [
          {arriveTangent=41.722, time=0, value=0.25}
          {interpMode=RCIM_Cubic, tangentMode=RCTM_User, time=0.02, value=1}
          {arriveTangent=-1.795, interpMode=RCIM_Cubic, leaveTangent=-1.795, tangentMode=RCTM_User, time=0.6, value=0.6}
          {arriveTangent=-3.611, arriveTangentWeight=0.112, interpMode=RCIM_Cubic, leaveTangent=-3.611, tangentMode=RCTM_User, time=1, value=0}
         ]}}})
    )
    Particles.DecalFadeSpray:NiagaraFloat=&Lerp_Float(
     A=0
     B=1
     Alpha=&FloatFromCurve(FloatCurve=DataInterface() {propertyValues {Curve {keys [
          {arriveTangent=41.722, time=0, value=1}
          {interpMode=RCIM_Cubic, tangentMode=RCTM_User, time=0.2, value=1}
          {arriveTangent=-0.5041, interpMode=RCIM_Cubic, leaveTangent=-0.5041, tangentMode=RCTM_User, time=1, value=0}
         ]}}})
    )
   ) {isSetParameters=true}
   /SetVariables(
    Particles.DecalSizeCrack:Vector3f=&Lerp_Vector(
     A=&Lerp_Vector(A=<100 250 250>, B=<100 400 400>, Alpha=$User.FX_Scale_Global)
     B=&Lerp_Vector(A=<100 500 500>, B=<100 800 800>, Alpha=$User.FX_Scale_Global)
     Alpha=&FloatFromCurve(FloatCurve=DataInterface() {propertyValues {Curve {keys [
          {leaveTangentWeight=0.6956, time=0, value=0}
          {arriveTangentWeight=0.0167, leaveTangentWeight=0.25, time=0.01, value=1}
          {arriveTangentWeight=0.4545, interpMode=RCIM_Cubic, leaveTangentWeight=0.2005, tangentMode=RCTM_User, time=0.6, value=1}
          {arriveTangent=-1.586, interpMode=RCIM_Cubic, leaveTangent=-1.586, tangentMode=RCTM_User, time=1, value=0.8}
         ]}}})
    )
    Particles.DecalSizeEmissive:Vector3f=&Lerp_Vector(
     A=<10 100 100>
     B=<10 700 700>
     Alpha=&FloatFromCurve(FloatCurve=DataInterface() {propertyValues {Curve {keys [
          {arriveTangent=8.3519, interpMode=RCIM_Cubic, leaveTangent=8.3519, tangentMode=RCTM_User, time=0, value=0}
          {arriveTangent=1.2843, interpMode=RCIM_Cubic, leaveTangent=1.2843, tangentMode=RCTM_User, time=0.1, value=0.7}
          {interpMode=RCIM_Cubic, tangentMode=RCTM_User, time=0.8, value=1}
         ]}}})
    )
    Particles.DecalSizeSpray:Vector3f=&Lerp_Vector(
     A=<100 200 100>
     B=<100 900 600>
     Alpha=&FloatFromCurve(FloatCurve=DataInterface() {propertyValues {Curve {keys [
          {interpMode=RCIM_Cubic, leaveTangentWeight=0.2519, tangentMode=RCTM_User, time=0, value=0}
          {arriveTangent=0.7686, interpMode=RCIM_Cubic, leaveTangent=0.7686, tangentMode=RCTM_User, time=0.05, value=0.95}
          {interpMode=RCIM_Cubic, tangentMode=RCTM_User, time=0.2, value=1}
         ]}}})
    )
   ) {isSetParameters=true}
  ]
  rendererClasses [&NiagaraDecalRendererProperties]
  renderers [
   {rendererClass=&NiagaraDecalRendererProperties, rendererIndex=0}
   {rendererClass=&NiagaraDecalRendererProperties, rendererIndex=1}
   {rendererClass=&NiagaraDecalRendererProperties, rendererIndex=2}
  ]
 }
 DirSplash {
  emitterUpdateScript [
   &EmitterState()
   &SpawnBurst_Instantaneous(Spawn Count=5, Use Spawn Probability=false, UseLoopCountLimit=false)
   /SetVariables(
    Emitter.FresnelColor:LinearColor=&SelectLinearColorFromArray(
     Array Sampling Mode=::NewEnumerator1
     Direct Array Index=$User.FXColor
     Color Selection Array=DataInterface() {propertyValues {ColorData [<rgba 1 1 1 1>, <rgba 0 0.6717 1 1>, <rgba 0.4253 0 1 1>, <rgba 0.0581 0.4948 0.0239 1>]}}
     Random Seed=?
    )
    Emitter.TileColor:LinearColor=&SelectLinearColorFromArray(
     Array Sampling Mode=::NewEnumerator1
     Direct Array Index=$User.FXColor
     Color Selection Array=DataInterface() {propertyValues {ColorData [<rgba 1 1 1 1>, <rgba 0 0.6717 1 1>, <rgba 0.5385 0.0009 1 1>, <rgba 0.1127 0.6146 0 1>]}}
     Random Seed=?
    )
   ) {isSetParameters=true}
   /SetVariables(Emitter.FresnelColor:LinearColor=<rgba 0.4253 0 1 1>, Emitter.TileColor:LinearColor=<rgba 0.5385 0.0009 1 1>) {enabled=false, isSetParameters=true}
  ]
  particleSpawnScript [
   &InitializeParticle(
    Lifetime=&RandomRangeFloat(Minimum=0.35, Maximum=0.5)
    Sprite Rotation Mode=::NewEnumerator0
    Sprite Size=<1 1>
    Sprite Size Mode=::NewEnumerator4
    UsePositionOffset=false
    Write Material Random=true
   )
   &AddVelocity(
    Cone Angle=200
    Cone Axis=<0 0 1>
    Module.UseConeFalloff=false
    Module.UseVelDistribution=false
    Override Seed=false
    Velocity Mode=::NewEnumerator2
    Velocity Speed=&Lerp_Float(
     A=800
     B=2000
     Alpha=&FloatFromCurve(
      FloatCurve=DataInterface() {propertyValues {Curve {keys [
          {arriveTangent=1.271, interpMode=RCIM_Cubic, leaveTangent=1.271, tangentMode=RCTM_User, time=0, value=0}
          {arriveTangent=3.4308, interpMode=RCIM_Cubic, leaveTangent=3.4308, tangentMode=RCTM_User, time=1, value=1}
         ]}}}
      CurveIndex=&RandomRangeFloat(Minimum=0, Maximum=1)
     )
    )
   )
   &ShapeLocation(MaskSpawnGroup=false, Override Seed=false, Sphere Radius=50, UseOwnerScale=true)
  ]
  particleUpdateScript [
   &ParticleState()
   &Drag(Drag=5, Use Linear Drag=true, Use Rotational Drag=false)
   &DynamicMaterialParameters(
    Index 0 Param 1=&FloatFromCurve(FloatCurve=DataInterface() {propertyValues {Curve {keys [
         <at 0 -1 RCIM_Cubic>
         {arriveTangent=2.3041, interpMode=RCIM_Cubic, leaveTangent=2.3041, tangentMode=RCTM_User, time=0.3, value=-0.675}
         <at 1 0.5>
        ]}}})
    Index 0 Param 2=&FloatFromCurve(FloatCurve=DataInterface() {propertyValues {Curve {keys [<at 0 3>,<at 0.3 1>]}}})
    Module.Param0WriteEnabled=true
    Module.Param1WriteEnabled=true
    Module.Param2WriteEnabled=true
    Module.Param3WriteEnabled=true
   )
   &ScaleSpriteSize(
    Scale Factor=&Lerp_Vector2(
     A=&Multiply_Vector2D_ByFloat(Float=2, Vector2D=&RandomRangeVector2D(Minimum=<25 275>, Maximum=<50 425>))
     B=&Multiply_Vector2D_ByFloat(Float=&RandomRangeFloat(Minimum=1, Maximum=2), Vector2D=<400 400>)
     Alpha=&FloatFromCurve(FloatCurve=DataInterface() {propertyValues {Curve {keys [
          {interpMode=RCIM_Cubic, leaveTangentWeight=1.1683, tangentMode=RCTM_User, time=0, value=0}
          {arriveTangent=0.8572, interpMode=RCIM_Cubic, leaveTangent=0.8572, tangentMode=RCTM_User, time=0.8, value=0.2}
          {arriveTangent=7.9683, arriveTangentWeight=0.9652, interpMode=RCIM_Cubic, leaveTangent=7.9683, tangentMode=RCTM_User, time=1, value=1}
         ]}}})
    )
    Scale Sprite Size Mode=::NewEnumerator2
   ) {module=ScaleSpriteSize001}
   &ScaleSpriteSize(
    Uniform Curve Scale=$User.FXScale
    Uniform Curve Sprite Scale=DataInterface() {propertyValues {Curve {keys [
        {arriveTangent=2.4725, interpMode=RCIM_Cubic, leaveTangent=2.4725, tangentMode=RCTM_User, time=0, value=0.3}
        {arriveTangent=0.6741, interpMode=RCIM_Cubic, leaveTangent=0.6741, tangentMode=RCTM_User, time=0.2, value=0.75}
        {tangentMode=RCTM_User, time=1, value=1}
       ]}}}
   )
   &GravityForce(Gravity=&RandomRangeVector(Minimum=<0 0 -25>, Maximum=<0 0 -200>))
   &SolveForcesAndVelocity(Clamp Velocity=false, Limit Acceleration=false)
   &ScaleColor(
    Scale Alpha=$System.TOD
    Scale RGB=&VectorFromFloat(Value=&Lerp_Float(
      A=0.5
      B=0.1
      Alpha=&FloatFromCurve(
       FloatCurve=DataInterface() {propertyValues {Curve {keys [
           {interpMode=RCIM_Cubic, leaveTangentWeight=0.966, tangentMode=RCTM_User, time=0, value=0}
           {arriveTangent=1.5389, interpMode=RCIM_Cubic, leaveTangent=1.5389, tangentMode=RCTM_User, time=0.5, value=0.6}
           {interpMode=RCIM_Cubic, tangentMode=RCTM_User, time=1, value=1}
          ]}}}
       CurveIndex=&RandomRangeFloat(Minimum=0, Maximum=1)
      )
     ))
    ScaleA=true
    ScaleRGB=true
   )
   &ScaleSpriteSizeBySpeed(Max Scale Factor=<1 3>, Min Scale Factor=<1 3>, Velocity Threshold=500) {enabled=false}
  ]
  rendererClasses [&NiagaraSpriteRendererProperties]
  renderers [{rendererClass=&NiagaraSpriteRendererProperties}]
 }
 Drops {
  emitterUpdateScript [
   &EmitterState(Enable Distance Culling=true, Module.bUseMaxDistance=true, Module.bUseMinDistance=false, Scalability Mode=::NewEnumerator1)
   &SpawnBurst_Instantaneous(Spawn Count=&UniformRangedInt(Minimum=50, Maximum=75), Use Spawn Probability=false)
   /SetVariables(
    Emitter.ColorA:LinearColor=&SelectLinearColorFromArray(
     Array Sampling Mode=::NewEnumerator1
     Direct Array Index=$User.FXColor
     Color Selection Array=DataInterface() {propertyValues {ColorData [<rgba 1 1 1 1>, <rgba 0 0.6717 1 1>, <rgba 0.1243 0 0.2656 1>, <rgba 0.0188 0.2656 0.0043 1>]}}
     Random Seed=?
    )
    Emitter.ColorB:LinearColor=&SelectLinearColorFromArray(
     Array Sampling Mode=::NewEnumerator1
     Direct Array Index=$User.FXColor
     Color Selection Array=DataInterface() {propertyValues {ColorData [<rgba 1 1 1 1>, <rgba 0 0.6717 1 1>, <rgba 0.4092 0 0.4167 1>, <rgba 0.0968 0.4167 0 1>]}}
     Random Seed=?
    )
    Emitter.ReflectionColor:LinearColor=&SelectLinearColorFromArray(
     Array Sampling Mode=::NewEnumerator1
     Direct Array Index=$User.FXColor
     Color Selection Array=DataInterface() {propertyValues {ColorData [<rgba 1 1 1 1>, <rgba 0 0.6717 1 1>, <rgba 3.9917 0 4 1>, <rgba 1.3599 4 0 1>]}}
     Random Seed=?
    )
   ) {isSetParameters=true}
  ]
  particleSpawnScript [
   &InitializeParticle(
    Lifetime=&Lerp_Float(
     A=0.5
     B=1
     Alpha=&FloatFromCurve(
      FloatCurve=DataInterface() {propertyValues {Curve {keys [
          {interpMode=RCIM_Cubic, leaveTangentWeight=0.966, tangentMode=RCTM_User, time=0, value=0}
          {arriveTangent=1, interpMode=RCIM_Cubic, leaveTangent=1, time=0.8, value=0.25}
          {arriveTangent=8.0065, interpMode=RCIM_Cubic, leaveTangent=8.0065, tangentMode=RCTM_User, time=1, value=1}
         ]}}}
      CurveIndex=&RandomRangeFloat(Minimum=0, Maximum=1)
     )
    )
    Sprite Size Mode=::NewEnumerator3
    Uniform Sprite Size=&Multiply_Float(A=&RandomRangeFloat(Minimum=20, Maximum=30), B=$User.FXScale)
    UsePositionOffset=false
    Write Material Random=true
   )
   &AddVelocity(
    Cone Angle=&Lerp_Float(
     A=70
     B=300
     Alpha=&FloatFromCurve(
      FloatCurve=DataInterface() {propertyValues {Curve {keys [
          {arriveTangent=1.271, interpMode=RCIM_Cubic, leaveTangent=1.271, tangentMode=RCTM_User, time=0, value=0}
          {arriveTangent=3.4308, interpMode=RCIM_Cubic, leaveTangent=3.4308, tangentMode=RCTM_User, time=1, value=1}
         ]}}}
      CurveIndex=&RandomRangeFloat(Minimum=0, Maximum=1)
     )
    )
    Cone Axis=<0 0 1>
    Module.UseConeFalloff=false
    Module.UseVelDistribution=false
    Override Seed=false
    Velocity Mode=::NewEnumerator2
    Velocity Speed=&Multiply_Float(
     A=&Lerp_Float(
      A=200
      B=2500
      Alpha=&FloatFromCurve(
       FloatCurve=DataInterface() {propertyValues {Curve {keys [
           {arriveTangent=1.271, interpMode=RCIM_Cubic, leaveTangent=1.271, tangentMode=RCTM_User, time=0, value=0}
           {arriveTangent=3.4308, interpMode=RCIM_Cubic, leaveTangent=3.4308, tangentMode=RCTM_User, time=1, value=1}
          ]}}}
       CurveIndex=&RandomRangeFloat(Minimum=0, Maximum=1)
      )
     )
     B=$User.FXScale
    )
   ) {module=AddVelocity002}
   &SphereLocation(
    MaskSpawnGroup=false
    Override Seed=false
    Sphere Radius=&RandomRangeFloat(Minimum=35, Maximum=60)
    Surface Only=false
    UseNonUniformScale=false
    UseOffset=false
    UseSphereOrientation=false
   )
  ]
  particleUpdateScript [
   &ParticleState()
   &GravityForce(Gravity=&RandomRangeVector(Minimum=<0 0 -600>, Maximum=<0 0 -1000>))
   &Drag(Drag=&UniformRangedFloat(Minimum=1, Maximum=1), Use Linear Drag=true, Use Rotational Drag=false)
   &ScaleColor(
    Scale Alpha=&Multiply_Float(A=&Lerp_Float(A=1, B=0, Alpha=$Particles.NormalizedAge), B=$System.TOD)
    Scale RGB=&VectorFromFloat(Value=&Lerp_Float(
      A=1
      B=0.1
      Alpha=&FloatFromCurve(
       FloatCurve=DataInterface() {propertyValues {Curve {keys [
           {arriveTangent=2.8697, interpMode=RCIM_Cubic, leaveTangent=2.8697, tangentMode=RCTM_User, time=0, value=0}
           {interpMode=RCIM_Cubic, tangentMode=RCTM_User, time=1, value=1}
          ]}}}
       CurveIndex=&RandomRangeFloat(Minimum=0, Maximum=1)
      )
     ))
    ScaleA=true
    ScaleRGB=true
   )
   &SolveForcesAndVelocity(Clamp Velocity=false, Limit Acceleration=false)
   &ScaleSpriteSizeBySpeed(Max Scale Factor=<1 5>, Min Scale Factor=<0 0.05>)
  ]
  rendererClasses [&NiagaraSpriteRendererProperties]
  renderers [{rendererClass=&NiagaraSpriteRendererProperties}]
 }
 FN_SourceParticleRibbon001 {
  emitterSpawnScript [/Game/Effects/Niagara/Modules/Attributes/FN_Source_v2(AttributeReader=DataInterface() {propertyValues {EmitterBinding {bindingMode=Other, emitterName=Projectile}}}) {enabled=false}]
  emitterUpdateScript [
   &EmitterState()
   &SpawnRate(
    SpawnRate=/Game/Effects/Niagara/DynamicInputs/Attributes/FN_GetEmitterParticleCount(AttributeReader:NiagaraDataInterfaceParticleRead=$Emitter.AttributeReader, Multiplier:NiagaraFloat=30)
    Use Spawn Probability=false
   )
   &FN_Source_v2(AttributeReader=DataInterface() {propertyValues {EmitterBinding {bindingMode=Other, emitterName=Projectile}}}) {module=FN_Source_v2001}
   /SetVariables(
    Emitter.HighlightColor:LinearColor=&SelectLinearColorFromArray(
     Array Sampling Mode=::NewEnumerator1
     Direct Array Index=$User.FXColor
     Color Selection Array=DataInterface() {propertyValues {ColorData [<rgba 1 1 1 1>, <rgba 0 0.6717 1 1>, <rgba 0.058 0 0.2708 1>, <rgba 0.0688 0.2708 0 1>]}}
     Random Seed=?
    )
    Emitter.LiquidColorA:LinearColor=&SelectLinearColorFromArray(
     Array Sampling Mode=::NewEnumerator1
     Direct Array Index=$User.FXColor
     Color Selection Array=DataInterface() {propertyValues {ColorData [<rgba 1 1 1 1>, <rgba 0 0.6717 1 1>, <rgba 0.0228 0 0.0677 1>, <rgba 0.0145 0.0677 0 1>]}}
     Random Seed=?
    )
    Emitter.LiquidColorB:LinearColor=&SelectLinearColorFromArray(
     Array Sampling Mode=::NewEnumerator1
     Direct Array Index=$User.FXColor
     Color Selection Array=DataInterface() {propertyValues {ColorData [<rgba 1 1 1 1>, <rgba 0 0.6717 1 1>, <rgba 0.0457 0 0.0938 1>, <rgba 0.0158 0.0938 0.0019 1>]}}
     Random Seed=?
    )
   ) {isSetParameters=true}
  ]
  particleSpawnScript [
   /Game/Effects/Niagara/Modules/Attributes/FN_SourceParentID(Ribbon ID from Parent ID=true)
   /Game/Effects/Niagara/Modules/Attributes/FN_SourceLocation(Position:Vector3f=/Game/Effects/Niagara/DynamicInputs/Attributes/FN_GetVectorByID(Attrib name:ENiagaraCommonVectors=::NewEnumerator0, Default on Invalid001:Vector3f=<0 0 0>))
   &InitializeRibbon(
    Module.Write Color=true
    Module.Write Lifetime=true
    Module.Write Mass=true
    Module.Write Material Random=true
    Module.Write Position=false
    Module.Write RibbonTwist=false
    Module.Write RibbonWidth=true
    Ribbon Width=&Multiply_Float(A=70, B=$User.FXScale)
   )
  ]
  particleUpdateScript [
   /Game/Effects/Niagara/Modules/Attributes/FN_UpdateSourcedRibbbons(
    Locked Particles Position:Vector3f=&FN_GetVectorByID(Attrib name:ENiagaraCommonVectors=::NewEnumerator0, Default on Invalid001:Vector3f=<0 0 0>)
    Ribbon Width:NiagaraFloat=$Particles.Initial.RibbonWidth
   ) {module=FN_UpdateLockedParticles}
   /SetVariables(Particles.Lifetime=&FloatFromCurve(
     FloatCurve=DataInterface() {propertyValues {Curve {keys [
         <at 0 1>
         {interpMode=RCIM_Cubic, tangentMode=RCTM_Break, time=0.5, value=1}
         {arriveTangent=-4.6273, interpMode=RCIM_Cubic, tangentMode=RCTM_Break, time=1, value=0}
        ]}}}
     Scale Curve=&FloatFromCurve(
      FloatCurve=DataInterface() {propertyValues {Curve {keys [
          {interpMode=RCIM_Cubic, tangentMode=RCTM_Break, time=0, value=1}
          {arriveTangent=-2.249, interpMode=RCIM_Cubic, tangentMode=RCTM_Break, time=1, value=0.2}
         ]}}}
      CurveIndex=&self:SampleParentNormalizedAge_01(
       Emitter:NiagaraDataInterfaceParticleRead=DataInterface() {propertyValues {EmitterBinding {bindingMode=Other, emitterName=Projectile}}}
       SampleID:NiagaraInt32=$Particles.SourceID
      )
     )
    )) {isSetParameters=true}
   &AccelerationForce(Acceleration=<0 0 -300>) {enabled=false}
   &SolveForcesAndVelocity(Clamp Velocity=false, Limit Acceleration=false)
   /SetVariables(Particles.RenderRibbonWidth:NiagaraFloat=&Multiply_Float(
     A=$Particles.Initial.RibbonWidth
     B=&FloatFromCurve(
      FloatCurve=DataInterface() {propertyValues {Curve {keys [
          {interpMode=RCIM_Cubic, leaveTangent=-0.0854, tangentMode=RCTM_Break, time=0, value=1}
          {arriveTangent=-3.6886, arriveTangentWeight=0.2821, interpMode=RCIM_Cubic, tangentMode=RCTM_Break, time=0.5, value=0}
         ]}}}
      Scale Curve=&FloatFromCurve(
       FloatCurve=DataInterface() {propertyValues {Curve {keys [
           {interpMode=RCIM_Cubic, tangentMode=RCTM_Break, time=0, value=1}
           {arriveTangent=-2.016, interpMode=RCIM_Cubic, tangentMode=RCTM_Break, time=1, value=0.2}
          ]}}}
       CurveIndex=$Particles.ParentAge
      )
     )
    )) {enabled=false, isSetParameters=true}
   /SetVariables(Particles.RenderRibbonWidth:NiagaraFloat=&Multiply_Float(
     A=$Particles.Initial.RibbonWidth
     B=&FloatFromCurve(
      FloatCurve=DataInterface() {propertyValues {Curve {keys [
          {interpMode=RCIM_Cubic, leaveTangent=-0.0854, tangentMode=RCTM_Break, time=0, value=1}
          {arriveTangent=-3.6886, arriveTangentWeight=0.2821, interpMode=RCIM_Cubic, tangentMode=RCTM_Break, time=0.5, value=0}
         ]}}}
      Scale Curve=&FloatFromCurve(
       FloatCurve=DataInterface() {propertyValues {Curve {keys [
           {interpMode=RCIM_Cubic, tangentMode=RCTM_Break, time=0, value=1}
           {
            arriveTangent=-0.74
            arriveTangentWeight=0.1
            interpMode=RCIM_Cubic
            leaveTangent=-0.74
            leaveTangentWeight=0.2658
            tangentMode=RCTM_User
            time=0.4
            value=0.9
           }
           {arriveTangent=-0.0479, interpMode=RCIM_Cubic, tangentMode=RCTM_Break, time=1, value=0.2}
          ]}}}
       CurveIndex=&self:SampleParentNormalizedAge_01(
        Emitter:NiagaraDataInterfaceParticleRead=DataInterface() {propertyValues {EmitterBinding {bindingMode=Other, emitterName=Projectile}}}
        SampleID:NiagaraInt32=$Particles.SourceID
       )
      )
     )
    )) {isSetParameters=true}
   &ScaleColor(
    Scale Alpha=&FloatFromCurve(
     FloatCurve=DataInterface() {propertyValues {Curve {keys [
         {interpMode=RCIM_Cubic, tangentMode=RCTM_Break, time=0.2, value=1}
         {arriveTangent=0.0059, arriveTangentWeight=0.3334, interpMode=RCIM_Cubic, tangentMode=RCTM_Break, time=1, value=0}
        ]}}}
     CurveIndex=$Particles.ParentAge
     Scale Curve=5
    )
    Scale RGB=&VectorFromFloat(Value=&RandomRangeFloat(Minimum=1, Maximum=2))
    ScaleA=true
    ScaleRGB=true
   )
  ]
  rendererClasses [&NiagaraRibbonRendererProperties]
  renderers [
   {rendererClass=&NiagaraRibbonRendererProperties}
   {rendererClass=&NiagaraRibbonRendererProperties, rendererIndex=1}
  ]
 }
 Juice {
  emitterUpdateScript [
   &EmitterState()
   &SpawnBurst_Instantaneous(Spawn Count=15, Use Spawn Probability=false, UseLoopCountLimit=false)
  ]
  particleSpawnScript [
   &InitializeParticle(
    Color=&HueShiftLinearColor(
     Hue=0
     Saturation=1
     Value=&RandomRangeFloat(Minimum=0.5, Maximum=1)
     Linear Color=&SelectLinearColorFromArray(
      Array Sampling Mode=::NewEnumerator1
      Direct Array Index=$User.FXColor
      Color Selection Array=DataInterface() {propertyValues {ColorData [<rgba 1 1 1 1>, <rgba 0 0.6717 1 1>, <rgba 0.2502 0 0.6146 1>, <rgba 0.0746 0.6146 0 1>]}}
      Random Seed=?
     )
    )
    Lifetime=&Lerp_Float(
     A=0.3
     B=0.7
     Alpha=&FloatFromCurve(
      FloatCurve=DataInterface() {propertyValues {Curve {keys [
          {arriveTangent=1.271, interpMode=RCIM_Cubic, leaveTangent=1.271, tangentMode=RCTM_User, time=0, value=0}
          {arriveTangent=3.4308, interpMode=RCIM_Cubic, leaveTangent=3.4308, tangentMode=RCTM_User, time=1, value=1}
         ]}}}
      CurveIndex=&RandomRangeFloat(Minimum=0, Maximum=1)
     )
    )
    Mesh Scale Mode=::NewEnumerator3
    Mesh Uniform Scale=&Lerp_Float(
     A=4
     B=4
     Alpha=&FloatFromCurve(
      FloatCurve=DataInterface() {propertyValues {Curve {keys [
          {arriveTangent=1.271, interpMode=RCIM_Cubic, leaveTangent=1.271, tangentMode=RCTM_User, time=0, value=0}
          {arriveTangent=3.4308, interpMode=RCIM_Cubic, leaveTangent=3.4308, tangentMode=RCTM_User, time=1, value=1}
         ]}}}
      CurveIndex=&RandomRangeFloat(Minimum=0, Maximum=1)
     )
    )
    UsePositionOffset=false
    Write Material Random=true
   )
   &InitializeParticle(
    Color Maximum=<rgba 0.2502 0 0.6146 1>
    Color Minimum=<rgba 0.0827 0 0.2031 1>
    Color Mode=::NewEnumerator2
    Lifetime=&Lerp_Float(
     A=0.3
     B=0.7
     Alpha=&FloatFromCurve(
      FloatCurve=DataInterface() {propertyValues {Curve {keys [
          {arriveTangent=1.271, interpMode=RCIM_Cubic, leaveTangent=1.271, tangentMode=RCTM_User, time=0, value=0}
          {arriveTangent=3.4308, interpMode=RCIM_Cubic, leaveTangent=3.4308, tangentMode=RCTM_User, time=1, value=1}
         ]}}}
      CurveIndex=&RandomRangeFloat(Minimum=0, Maximum=1)
     )
    )
    Mesh Scale Mode=::NewEnumerator3
    Mesh Uniform Scale=&Lerp_Float(
     A=4
     B=4
     Alpha=&FloatFromCurve(
      FloatCurve=DataInterface() {propertyValues {Curve {keys [
          {arriveTangent=1.271, interpMode=RCIM_Cubic, leaveTangent=1.271, tangentMode=RCTM_User, time=0, value=0}
          {arriveTangent=3.4308, interpMode=RCIM_Cubic, leaveTangent=3.4308, tangentMode=RCTM_User, time=1, value=1}
         ]}}}
      CurveIndex=&RandomRangeFloat(Minimum=0, Maximum=1)
     )
    )
    UsePositionOffset=false
    Write Material Random=true
   ) {enabled=false, module=InitializeParticle001}
   &ShapeLocation(
    MaskSpawnGroup=false
    Override Seed=false
    Sphere Radius=&Lerp_Float(
     A=15
     B=75
     Alpha=&FloatFromCurve(
      FloatCurve=DataInterface() {propertyValues {Curve {keys [
          {arriveTangent=1.271, interpMode=RCIM_Cubic, leaveTangent=1.271, tangentMode=RCTM_User, time=0, value=0}
          {arriveTangent=3.4308, interpMode=RCIM_Cubic, leaveTangent=3.4308, tangentMode=RCTM_User, time=1, value=1}
         ]}}}
      CurveIndex=&RandomRangeFloat(Minimum=0, Maximum=1)
     )
    )
    UseOwnerScale=true
   )
   &AddVelocity(
    Cone Angle=&Lerp_Float(
     A=100
     B=300
     Alpha=&FloatFromCurve(
      FloatCurve=DataInterface() {propertyValues {Curve {keys [
          {arriveTangent=1.271, interpMode=RCIM_Cubic, leaveTangent=1.271, tangentMode=RCTM_User, time=0, value=0}
          {arriveTangent=3.4308, interpMode=RCIM_Cubic, leaveTangent=3.4308, tangentMode=RCTM_User, time=1, value=1}
         ]}}}
      CurveIndex=&RandomRangeFloat(Minimum=0, Maximum=1)
     )
    )
    Cone Axis=<0 0 1>
    Module.UseConeFalloff=false
    Module.UseVelDistribution=false
    Override Seed=false
    Velocity Mode=::NewEnumerator2
    Velocity Speed=&Lerp_Float(
     A=200
     B=1500
     Alpha=&FloatFromCurve(
      FloatCurve=DataInterface() {propertyValues {Curve {keys [
          {arriveTangent=1.271, interpMode=RCIM_Cubic, leaveTangent=1.271, tangentMode=RCTM_User, time=0, value=0}
          {arriveTangent=3.4308, interpMode=RCIM_Cubic, leaveTangent=3.4308, tangentMode=RCTM_User, time=1, value=1}
         ]}}}
      CurveIndex=&RandomRangeFloat(Minimum=0, Maximum=1)
     )
    )
   )
   &AddVelocity(Velocity=&RandomRangeVector(Minimum=<0 0 300>, Maximum=<0 0 300>)) {enabled=false, module=AddVelocity001}
   &AddVelocity(
    Cone Angle=&Lerp_Float(
     A=60
     B=200
     Alpha=&FloatFromCurve(
      FloatCurve=DataInterface() {propertyValues {Curve {keys [
          {arriveTangent=1.271, interpMode=RCIM_Cubic, leaveTangent=1.271, tangentMode=RCTM_User, time=0, value=0}
          {arriveTangent=3.4308, interpMode=RCIM_Cubic, leaveTangent=3.4308, tangentMode=RCTM_User, time=1, value=1}
         ]}}}
      CurveIndex=&RandomRangeFloat(Minimum=0, Maximum=1)
     )
    )
    Module.UseConeFalloff=false
    Module.UseVelDistribution=false
    Override Seed=false
    Velocity Mode=::NewEnumerator2
    Velocity Speed=&Lerp_Float(
     A=200
     B=2500
     Alpha=&FloatFromCurve(
      FloatCurve=DataInterface() {propertyValues {Curve {keys [
          {arriveTangent=1.271, interpMode=RCIM_Cubic, leaveTangent=1.271, tangentMode=RCTM_User, time=0, value=0}
          {arriveTangent=3.4308, interpMode=RCIM_Cubic, leaveTangent=3.4308, tangentMode=RCTM_User, time=1, value=1}
         ]}}}
      CurveIndex=&RandomRangeFloat(Minimum=0, Maximum=1)
     )
    )
   ) {enabled=false, module=AddVelocity002}
   /SetVariables(Particles.Velocity=&Multiply_Vector(A=$Particles.Velocity, B=<1 1 1>)) {isSetParameters=true}
   &CameraOffset(Camera Offset Amount=20) {enabled=false}
  ]
  particleUpdateScript [
   &ParticleState()
   &Drag(Use Linear Drag=true, Use Rotational Drag=false)
   &GravityForce(Gravity=&RandomRangeVector(Minimum=<0 0 -50>, Maximum=<0 0 -300>))
   &SolveForcesAndVelocity(Clamp Velocity=false, Limit Acceleration=false)
   &ScaleMeshSize(Scale Factor=&VectorFromFloat(Value=&FloatFromCurve(
      FloatCurve=DataInterface() {propertyValues {Curve {keys [
          {arriveTangent=1.704, interpMode=RCIM_Cubic, leaveTangent=1.704, tangentMode=RCTM_User, time=0, value=0.1}
          {arriveTangent=0.2614, interpMode=RCIM_Cubic, leaveTangent=0.2614, tangentMode=RCTM_User, time=1, value=1}
         ]}}}
      Scale Curve=$User.FXScale
     )))
   &DynamicMaterialParameters(
    Index 0 Param 1=&FloatFromCurve(FloatCurve=DataInterface() {propertyValues {Curve {keys [<at 0 0 RCIM_Cubic>, {arriveTangent=0.9849, interpMode=RCIM_Cubic, leaveTangent=0.9849, tangentMode=RCTM_User, time=1, value=1}]}}})
    Index 0 Param 2=&FloatFromCurve(
     FloatCurve=DataInterface() {propertyValues {Curve {keys [
         {arriveTangent=1.3993, interpMode=RCIM_Cubic, leaveTangent=1.3993, tangentMode=RCTM_User, time=0, value=0.5}
         {interpMode=RCIM_Cubic, tangentMode=RCTM_User, time=1, value=1}
        ]}}}
     Scale Curve=&Multiply_Float(A=1.5, B=1)
    )
    Index 0 Param 3=&Multiply_Float(
     A=&MakeFloatFromVector(Vector=$Particles.Scale, Channel=::NewEnumerator0)
     B=&RandomRangeFloat(Minimum=2, Maximum=3)
    )
    Index 0 Param 4=&FloatFromCurve(
     FloatCurve=DataInterface() {propertyValues {Curve {keys [{arriveTangent=1.8524, interpMode=RCIM_Cubic, time=0, value=0}, {interpMode=RCIM_Cubic, leaveTangent=1.8693, time=1, value=1}]}}}
     Scale Curve=&RandomRangeFloat(Minimum=0.5, Maximum=1.5)
    )
    Index 1 Param 1=&FloatFromCurve(
     FloatCurve=DataInterface() {propertyValues {Curve {keys [
         {arriveTangent=2.4162, interpMode=RCIM_Cubic, leaveTangent=2.4162, tangentMode=RCTM_User, time=0, value=0}
         {interpMode=RCIM_Cubic, tangentMode=RCTM_User, time=1, value=1}
        ]}}}
     Scale Curve=-5
    )
    Index 1 Param 2=&RandomRangeFloat(Minimum=0.5, Maximum=1)
    Module.Param0WriteEnabled=true
    Module.Param0WriteEnabled1=true
    Module.Param1WriteEnabled=true
    Module.Param1WriteEnabled1=true
    Module.Param2WriteEnabled=true
    Module.Param2WriteEnabled1=false
    Module.Param3WriteEnabled=true
    Module.Param3WriteEnabled1=false
    Write Parameter Index 1=true
   )
   &ScaleColor(
    Scale Alpha=&Lerp_Float(
     A=5
     B=10
     Alpha=&FloatFromCurve(
      FloatCurve=DataInterface() {propertyValues {Curve {keys [
          {interpMode=RCIM_Cubic, leaveTangentWeight=1.0035, tangentMode=RCTM_User, time=0.2, value=0}
          {arriveTangent=3.9436, interpMode=RCIM_Cubic, leaveTangent=3.9436, tangentMode=RCTM_User, time=1, value=1}
         ]}}}
      CurveIndex=&RandomRangeFloat(Minimum=0, Maximum=1)
     )
    )
    Scale RGB=&VectorFromFloat(Value=&Lerp_Float(
      A=&RandomRangeFloat(Minimum=0.1, Maximum=0.5)
      B=1
      Alpha=&FloatFromCurve(FloatCurve=DataInterface() {propertyValues {Curve {keys [
           {arriveTangent=-3.6277, interpMode=RCIM_Cubic, leaveTangent=-3.6277, tangentMode=RCTM_User, time=0.2, value=1}
           {interpMode=RCIM_Cubic, tangentMode=RCTM_User, time=1, value=0}
          ]}}})
     ))
    ScaleA=true
    ScaleRGB=true
   )
  ]
  rendererClasses [&NiagaraMeshRendererProperties]
  renderers [{rendererClass=&NiagaraMeshRendererProperties, rendererIndex=0}]
 }
 Projectile {
  emitterUpdateScript [
   &EmitterState()
   &SpawnBurst_Instantaneous(Spawn Count=5, Use Spawn Probability=false, UseLoopCountLimit=false)
   &SpawnRate(SpawnRate=5, Use Spawn Probability=false) {enabled=false}
  ]
  particleSpawnScript [
   &InitializeParticle(
    Color Maximum=<rgba 0.0422 0 0.1146 1>
    Color Minimum=<rgba 0.0121 0 0.0469 1>
    Color Mode=::NewEnumerator2
    Lifetime=&Multiply_Float(
     A=&RandomRangeFloat(Minimum=2, Maximum=3)
     B=&FloatFromCurve(
      FloatCurve=DataInterface() {propertyValues {Curve {keys [{interpMode=RCIM_Cubic, tangentMode=RCTM_Break, time=0, value=0.4}, {interpMode=RCIM_Cubic, tangentMode=RCTM_Break, time=1, value=0.7}]}}}
      CurveIndex=$Emitter.Age
     )
    )
    Mesh Scale Max=<1 1 1>
    Mesh Scale Min=<0.5 0.5 0.5>
    Mesh Scale Mode=::NewEnumerator2
    UsePositionOffset=false
    Write Material Random=true
   )
   &DynamicMaterialParameters(
    Index 0 Param 1=50
    Index 0 Param 2=1
    Index 0 Param 4=&RandomRangeFloat(Minimum=-100, Maximum=100)
    Index 1 Param 1=&RandomRangeFloat(Minimum=-10, Maximum=5)
    Index 1 Param 2=1
    Index 1 Param 3=1
    Module.Param0WriteEnabled=true
    Module.Param0WriteEnabled1=true
    Module.Param1WriteEnabled=true
    Module.Param1WriteEnabled1=true
    Module.Param2WriteEnabled=true
    Module.Param2WriteEnabled1=true
    Module.Param3WriteEnabled=true
    Module.Param3WriteEnabled1=false
    Write Parameter Index 1=true
   ) {module=DynamicMaterialParameters001}
   &InitialMeshOrientation(Rotation=&RandomRangeVector(Minimum=<0 0 0>, Maximum=<1 1 1>), Use Rotation Vector=true)
   &ShapeLocation(MaskSpawnGroup=false, Override Seed=false, Sphere Radius=35, UseOwnerScale=true)
   &AddVelocity(
    Cone Angle=120
    Cone Axis=<0 0 1>
    Module.UseConeFalloff=false
    Module.UseVelDistribution=false
    Override Seed=false
    Velocity Mode=::NewEnumerator2
    Velocity Speed=&Multiply_Float(
     A=&Lerp_Float(
      A=600
      B=1200
      Alpha=&FloatFromCurve(
       FloatCurve=DataInterface() {propertyValues {Curve {keys [
           {arriveTangent=1.271, interpMode=RCIM_Cubic, leaveTangent=1.271, tangentMode=RCTM_User, time=0, value=0}
           {arriveTangent=3.4308, interpMode=RCIM_Cubic, leaveTangent=3.4308, tangentMode=RCTM_User, time=1, value=1}
          ]}}}
       CurveIndex=&RandomRangeFloat(Minimum=0, Maximum=1)
      )
     )
     B=$User.FXScale
    )
   ) {module=AddVelocity002}
  ]
  particleUpdateScript [
   &ParticleState()
   &Drag(Drag=2, Use Linear Drag=true, Use Rotational Drag=false) {enabled=false}
   &AccelerationForce(Acceleration=<0 0 -1600>)
   &Collision(
    Advanced Aging Rate=1
    Age Colliding Particles=false
    Module.Control Rest State Via Surface Normal=false
    Module.EnableMaxCollisionCount=false
    Module.Randomize Collision Normal Vector=true
    Particle Radius Scale=5
    Radius Calculation Type=::NewEnumerator1
    Restitution=0.5
   ) {module=Collision001}
   &SolveForcesAndVelocity(
    Clamp Velocity=true
    Limit Acceleration=false
    Speed Limit:NiagaraFloat=&FloatFromCurve(FloatCurve=DataInterface() {propertyValues {Curve {keys [<at 4 20000>,<at 10 5000>]}}}, CurveIndex=$Particles.Age)
   )
   &ScaleMeshSize(Scale Factor=&VectorFromFloat(Value=&FloatFromCurve(
      FloatCurve=DataInterface() {propertyValues {Curve {keys [
          {arriveTangent=10, interpMode=RCIM_Cubic, time=0, value=0}
          {interpMode=RCIM_Cubic, tangentMode=RCTM_User, time=0.1, value=1}
          {interpMode=RCIM_Cubic, tangentMode=RCTM_User, time=0.4, value=1}
          {arriveTangent=-4.0437, interpMode=RCIM_Cubic, leaveTangent=-4.0437, tangentMode=RCTM_User, time=1, value=0}
         ]}}}
      Scale Curve=0
     )))
   &UpdateMeshOrientation(
    Rotation Rate=&RandomRangeFloat(Minimum=-1, Maximum=1)
    Rotation Vector=&RandomRangeVector(Minimum=<0 0 0>, Maximum=<1 1 1>)
   )
   &ScaleColor(Scale RGB=&VectorFromFloat(Value=0.5), ScaleA=false, ScaleRGB=true)
  ]
  rendererClasses [&NiagaraMeshRendererProperties]
  renderers [{rendererClass=&NiagaraMeshRendererProperties, rendererIndex=0}]
 }
 Rainbow_Smoke {
  emitterUpdateScript [
   &EmitterState()
   &SpawnBurst_Instantaneous(Spawn Count=7, Use Spawn Probability=false, UseLoopCountLimit=false)
  ]
  particleSpawnScript [
   &InitializeParticle(
    Color=&HueShiftLinearColor(
     Hue=0
     Saturation=1
     Value=&RandomRangeFloat(Minimum=0.5, Maximum=1)
     Linear Color=&SelectLinearColorFromArray(
      Array Sampling Mode=::NewEnumerator1
      Direct Array Index=$User.FXColor
      Color Selection Array=DataInterface() {propertyValues {ColorData [<rgba 1 1 1 1>, <rgba 0 0.6717 1 1>, <rgba 0.4616 0 1 1>, <rgba 0.1648 0.7708 0.0449 1>]}}
      Random Seed=?
     )
    )
    Lifetime Max=0.6
    Lifetime Min=0.4
    Lifetime Mode=::NewEnumerator1
    Sprite Rotation Mode=::NewEnumerator2
    Sprite Size Mode=::NewEnumerator1
    Uniform Sprite Size Max=800
    Uniform Sprite Size Min=600
    UsePositionOffset=false
    Write Material Random=true
   )
   &InitializeParticle(
    Color Maximum=<rgba 0.4616 0 1 1>
    Color Minimum=<rgba 0.2966 0 1 1>
    Color Mode=::NewEnumerator2
    Lifetime Max=0.6
    Lifetime Min=0.4
    Lifetime Mode=::NewEnumerator1
    Sprite Rotation Mode=::NewEnumerator2
    Sprite Size Mode=::NewEnumerator1
    Uniform Sprite Size Max=800
    Uniform Sprite Size Min=600
    UsePositionOffset=false
    Write Material Random=true
   ) {enabled=false, module=InitializeParticle001}
   &DynamicMaterialParameters(
    Index 0 Param 1=&RandomRangeFloat(Minimum=0, Maximum=1)
    Module.Param0WriteEnabled=true
    Module.Param1WriteEnabled=true
    Module.Param2WriteEnabled=true
    Module.Param3WriteEnabled=true
   )
   &AddVelocity(
    Cone Angle=200
    Cone Axis=<0 0 1>
    Module.UseConeFalloff=false
    Module.UseVelDistribution=false
    Override Seed=false
    Velocity Mode=::NewEnumerator2
    Velocity Speed=&Lerp_Float(
     A=300
     B=1000
     Alpha=&FloatFromCurve(
      FloatCurve=DataInterface() {propertyValues {Curve {keys [
          {arriveTangent=1.271, interpMode=RCIM_Cubic, leaveTangent=1.271, tangentMode=RCTM_User, time=0, value=0}
          {arriveTangent=3.4308, interpMode=RCIM_Cubic, leaveTangent=3.4308, tangentMode=RCTM_User, time=1, value=1}
         ]}}}
      CurveIndex=&RandomRangeFloat(Minimum=0, Maximum=1)
     )
    )
   )
   &ShapeLocation(MaskSpawnGroup=false, Override Seed=false, Sphere Radius=50, UseOwnerScale=true)
  ]
  particleUpdateScript [
   &ParticleState()
   &SpriteRotationRate(Rotation Rate=&RandomRangeFloat(Minimum=25, Maximum=100))
   &ScaleSpriteSize(
    Uniform Curve Scale=$User.FXScale
    Uniform Curve Sprite Scale=DataInterface() {propertyValues {Curve {keys [
        {interpMode=RCIM_Cubic, tangentMode=RCTM_User, time=0, value=0.3}
        {arriveTangent=1.167, interpMode=RCIM_Cubic, leaveTangent=1.167, tangentMode=RCTM_User, time=1, value=1}
       ]}}}
   )
   &Drag(Use Linear Drag=true, Use Rotational Drag=true)
   &SolveForcesAndVelocity(Clamp Velocity=false, Limit Acceleration=false)
   &ScaleColor(
    Scale Alpha=&Multiply_Float(
     A=&FloatFromCurve(
      FloatCurve=DataInterface() {propertyValues {Curve {keys [
          {interpMode=RCIM_Cubic, tangentMode=RCTM_User, time=0, value=1}
          {arriveTangentWeight=0.7021, interpMode=RCIM_Cubic, tangentMode=RCTM_User, time=1, value=0}
         ]}}}
      Scale Curve=&RandomRangeFloat(Minimum=0.5, Maximum=1.5)
     )
     B=&Multiply_Float(A=1, B=$System.TOD)
    )
    Scale RGB=&VectorFromFloat(Value=&RandomRangeFloat(Minimum=1, Maximum=3))
    ScaleA=true
    ScaleRGB=true
   )
  ]
  rendererClasses [&NiagaraSpriteRendererProperties]
  renderers [{rendererClass=&NiagaraSpriteRendererProperties}]
 }
 Splash {
  emitterUpdateScript [
   &EmitterState()
   &SpawnBurst_Instantaneous(Spawn Count=9, Use Spawn Probability=false)
  ]
  particleSpawnScript [
   &InitializeParticle(
    Color=&HueShiftLinearColor(
     Hue=0
     Saturation=1
     Value=&RandomRangeFloat(Minimum=0.5, Maximum=1)
     Linear Color=&SelectLinearColorFromArray(
      Array Sampling Mode=::NewEnumerator1
      Direct Array Index=$User.FXColor
      Color Selection Array=DataInterface() {propertyValues {ColorData [<rgba 1 1 1 1>, <rgba 0 0.6717 1 1>, <rgba 0.1354 0 0.1264 1>, <rgba 0.0195 0.1354 0 1>]}}
      Random Seed=?
     )
    )
    Lifetime Max=0.8
    Lifetime Min=0.4
    Lifetime Mode=::NewEnumerator1
    Sprite Rotation Mode=::NewEnumerator2
    Sprite Size Mode=::NewEnumerator1
    Uniform Sprite Size Max=30
    Uniform Sprite Size Min=20
    UsePositionOffset=false
    Write Material Random=true
   )
   &InitializeParticle(
    Color Maximum=<rgba 0.1354 0 0.1264 1>
    Color Minimum=<rgba 0.0372 0 0.1042 1>
    Color Mode=::NewEnumerator2
    Lifetime Max=0.8
    Lifetime Min=0.4
    Lifetime Mode=::NewEnumerator1
    Sprite Rotation Mode=::NewEnumerator2
    Sprite Size Mode=::NewEnumerator1
    Uniform Sprite Size Max=30
    Uniform Sprite Size Min=20
    UsePositionOffset=false
    Write Material Random=true
   ) {enabled=false, module=InitializeParticle001}
   &AddVelocity(
    Cone Angle=&Lerp_Float(
     A=75
     B=250
     Alpha=&FloatFromCurve(
      FloatCurve=DataInterface() {propertyValues {Curve {keys [
          {arriveTangent=1.271, interpMode=RCIM_Cubic, leaveTangent=1.271, tangentMode=RCTM_User, time=0, value=0}
          {arriveTangent=3.4308, interpMode=RCIM_Cubic, leaveTangent=3.4308, tangentMode=RCTM_User, time=1, value=1}
         ]}}}
      CurveIndex=&RandomRangeFloat(Minimum=0, Maximum=1)
     )
    )
    Cone Axis=<0 0 1>
    Module.UseConeFalloff=false
    Module.UseVelDistribution=false
    Override Seed=false
    Velocity Mode=::NewEnumerator2
    Velocity Speed=&Multiply_Float(
     A=&Lerp_Float(
      A=100
      B=2500
      Alpha=&FloatFromCurve(
       FloatCurve=DataInterface() {propertyValues {Curve {keys [
           {interpMode=RCIM_Cubic, leaveTangentWeight=0.3555, tangentMode=RCTM_User, time=0, value=0}
           {arriveTangent=3.4308, interpMode=RCIM_Cubic, leaveTangent=3.4308, tangentMode=RCTM_User, time=1, value=1}
          ]}}}
       CurveIndex=&RandomRangeFloat(Minimum=0, Maximum=1)
      )
     )
     B=$User.FXScale
    )
   )
   &ShapeLocation(
    MaskSpawnGroup=false
    Override Seed=false
    Sphere Radius=&RandomRangeFloat(Minimum=25, Maximum=60)
    UseOwnerScale=true
   )
  ]
  particleUpdateScript [
   &ParticleState()
   &Drag(Use Linear Drag=true, Use Rotational Drag=false)
   &GravityForce(Coordinate Space=::Local, Gravity=<0 0 -1000>)
   &AccelerationForce(Acceleration=<0 0 -200>) {enabled=false}
   &SolveForcesAndVelocity(Clamp Velocity=false, Limit Acceleration=false)
   &SpriteRotationRate(Rotation Rate=&RandomRangeFloat(Minimum=300, Maximum=-300))
   &ScaleSpriteSize(Scale Factor=&Vector2DFromFloat(Value=&FloatFromCurve(
      FloatCurve=DataInterface() {propertyValues {Curve {keys [
          {interpMode=RCIM_Cubic, tangentMode=RCTM_User, time=0, value=0}
          {interpMode=RCIM_Cubic, tangentMode=RCTM_User, time=0.25, value=1}
          {interpMode=RCIM_Cubic, tangentMode=RCTM_User, time=0.8, value=1}
          {interpMode=RCIM_Cubic, tangentMode=RCTM_User, time=1, value=0}
         ]}}}
      Scale Curve=&Multiply_Float(A=5, B=$User.FXScale)
     )))
   &ScaleColor(
    Scale Alpha=&FloatFromCurve(
     FloatCurve=DataInterface() {propertyValues {Curve {keys [<at 0.2 1>,<at 0.2 1>,<at 1 0>]}}}
     Scale Curve=&FloatFromCurve(
      FloatCurve=DataInterface() {propertyValues {Curve {keys [
          {interpMode=RCIM_Cubic, tangentMode=RCTM_User, time=0, value=1}
          {arriveTangent=-1.5991, interpMode=RCIM_Cubic, leaveTangent=-1.5991, tangentMode=RCTM_User, time=1, value=0}
         ]}}}
      Scale Curve=$System.TOD
     )
    )
    ScaleA=true
    ScaleRGB=false
   )
   &CameraOffset(Camera Offset Amount=&RandomRangeFloat(Minimum=2, Maximum=3))
   &DynamicMaterialParameters(Module.Param0WriteEnabled=true, Module.Param1WriteEnabled=true, Module.Param2WriteEnabled=true, Module.Param3WriteEnabled=true)
  ]
  rendererClasses [&NiagaraSpriteRendererProperties]
  renderers [{rendererClass=&NiagaraSpriteRendererProperties}]
 }
}
summary {
 systemName=NS_Proto_GooItem_Burst_Ball
 emitters [
  {bEnabled=false, emitterName=DEBUG, rendererClasses [&NiagaraSpriteRendererProperties]}
  {emitterName=Drops, rendererClasses [&NiagaraSpriteRendererProperties]}
  {emitterName=DirSplash, rendererClasses [&NiagaraSpriteRendererProperties]}
  {emitterName=Juice, rendererClasses [&NiagaraMeshRendererProperties]}
  {emitterName=Splash, rendererClasses [&NiagaraSpriteRendererProperties]}
  {emitterName=Rainbow_Smoke, rendererClasses [&NiagaraSpriteRendererProperties]}
  {emitterName=CenterSplash, rendererClasses [&NiagaraSpriteRendererProperties]}
  {emitterName=Projectile, rendererClasses [&NiagaraMeshRendererProperties]}
  {emitterName=FN_SourceParticleRibbon001, rendererClasses [&NiagaraRibbonRendererProperties]}
  {emitterName=DecalRenderer, rendererClasses [&NiagaraDecalRendererProperties]}
  {emitterName=BrightStreaks, rendererClasses [&NiagaraSpriteRendererProperties]}
 ]
 userVariables [
  User.FX_Scale_Global {type=NiagaraFloat, defaultValue=0, description=""}
  User.FXColor {type=NiagaraInt32, defaultValue=2}
  User.FXDecal {type=NiagaraBool, defaultValue=true}
  User.FXScale {type=NiagaraFloat, defaultValue=1, description=""}
  User.FXTime {type=NiagaraFloat, defaultValue=0.5, description=""}
 ]
}
